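# Physics asset for the back-hung sword belt on the DreamPaneer male medium character cosmetic.
# A dynamic belt capsule hangs from a small base capsule through one constraint with limited swing and no twist.
# Four skirt capsules, named as colliders, have no constraints and appear to act as collision shapes for the swinging belt.
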PhysicsAsset /BRCosmetics/Characters/Player/Male/Medium/Bodies/M_MED_DreamPaneer/Meshes/M_MED_DreamPaneer_BackSword_Physics
bodies {
 base_backSwordHangBelt [{shapeType=Capsule, length=0.5, radius=0.25, shapeName=base_sideBag_capsule}]
 coll_skirt_01_c [{shapeType=Capsule, length=18.6399, radius=4.0718, shapeName=coll_skirt_01_c_capsule, center=<11.3705 0 -3.3972>, rotation=<rot 90 0 0>}]
 coll_skirt_01_l [{shapeType=Capsule, length=18.8368, radius=4.0718, shapeName=coll_skirt_01_l_capsule, center=<10.7428 0 -2.1352>, rotation=<rot 87.1404 0 0>}]
 coll_skirt_01_r [{
   shapeType=Capsule
   length=17.8241
   radius=4.0718
   shapeName=coll_skirt_01_r_capsule
   center=<10.7289 -0.0473 -3.7235>
   rotation=<rot 87.4703 7.1449 6.723>
  }]
 coll_skirt_02_r [{
   shapeType=Capsule
   length=17.1261
   radius=4.9729
   shapeName=coll_skirt_02_r_capsule
   center=<10.7075 0.4805 -2.7018>
   rotation=<rot 82.2867 56.1137 52.4848>
  }]
 dyn_backSwordHangBelt [{shapeType=Capsule, length=23.4398, radius=4.6333, shapeName=dyn_backSwordHangBelt_capsule, center=<13.7643 0 -2.5461>, rotation=<rot 0 0 90>}]
}
constraints [{bone1Name=dyn_backSwordHangBelt, bone2Name=base_backSwordHangBelt, swing1LimitDegrees=30, swing2LimitDegrees=35, twistLimitDegrees=0}]
