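# Physics asset for the skirt accessories on the DreamPaneer medium male body, covering the dangling potion, side bag, side sword, hip chain and back-sword hang belt.
# Capsules on spine, thighs and skirt collision bones provide body shapes around the accessories.
# Potion, bag and sword use swing-limited, zero-twist constraints to small base bodies; the two-link hip chain has tighter limits.
# Angularly free constraints tie those three bases and the back sword to spine_01, and the belt to the back sword.
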PhysicsAsset /BRCosmetics/Characters/Player/Male/Medium/Bodies/M_MED_DreamPaneer/Meshes/M_MED_DreamPaneer_SkirtAcc_Physics
bodies {
 base_hipChain_1 [{shapeType=Capsule, length=0.5, radius=0.25, shapeName=base_sideBag_capsule}]
 base_potion [{shapeType=Capsule, length=0.5, radius=0.25, shapeName=base_sideBag_capsule}]
 base_sideBag [{shapeType=Capsule, length=0.5, radius=0.25, shapeName=base_sideBag_capsule}]
 base_sideSword [{shapeType=Capsule, length=0.5, radius=0.25, shapeName=base_sideBag_capsule}]
 coll_skirt_02_c [{shapeType=Capsule, length=13.6208, radius=5.4398, shapeName=coll_skirt_02_r_capsule, center=<7.883 0.2833 -4.5403>, rotation=<rot 90 0 0>}]
 coll_skirt_02_l [{shapeType=Capsule, length=14.8129, radius=5.374, shapeName=coll_skirt_02_r_capsule, center=<7.8794 -0.2982 -3.6404>, rotation=<rot 90 0 0>}]
 coll_skirt_02_r [{shapeType=Capsule, length=13.2628, radius=5.7978, shapeName=coll_skirt_02_r_capsule, center=<11.6901 -0.7752 -4.2318>, rotation=<rot 90 0 0>}]
 coll_skirt_03_r [{
   shapeType=Capsule
   length=14.16
   radius=4.9006
   shapeName=coll_skirt_03_r_capsule
   center=<11.5789 1.0259 -3.5498>
   rotation=<rot 86.6439 0 0>
  }]
 dyn_backSwordHangBelt [{shapeType=Capsule, length=18.9065, radius=4.5668, shapeName=dyn_backSwordHangBelt_capsule, center=<6.4051 0 0>, rotation=<rot 0 0 90>}]
 dyn_hipChain_1 [{shapeType=Capsule, length=5.4383, radius=1.6786, shapeName=dyn_hipChain_1_capsule, center=<3.4727 0 0.571>, rotation=<rot -90 0 0>}]
 dyn_hipChain_2 [{shapeType=Capsule, length=5.4844, radius=1.6175, shapeName=dyn_hipChain_2_capsule, center=<3.6413 0 0.6193>, rotation=<rot -90 0 0>}]
 dyn_potion [
  {
   shapeType=Capsule
   length=2.1856
   radius=4.5589
   shapeName=dyn_potion_capsule
   center=<11.4129 -0.6086 0.8285>
   rotation=<rot 0 -76.8281 -82.136>
  }
  {
   shapeType=Capsule
   length=3.7763
   radius=2.9682
   shapeName=dyn_potion_capsuleA
   center=<3.3427 -2.4973 1.9733>
   rotation=<rot 0 -76.8281 -82.136>
  }
 ]
 dyn_sideBag [
  {
   shapeType=Capsule
   length=12.3344
   radius=4.2853
   shapeName=dyn_sideBag_capsule
   center=<7.43 -1.085 4.4516>
   rotation=<rot -88.2545 -179.9954 -179.9997>
  }
  {
   shapeType=Capsule
   length=12.3344
   radius=4.2853
   shapeName=dyn_sideBag_capsuleA
   center=<7.4298 1.511 4.4516>
   rotation=<rot -88.2545 -179.9954 -179.9997>
  }
 ]
 dyn_sideSword [{
   shapeType=Capsule
   length=37.4825
   radius=4.4209
   shapeName=dyn_sideSword_capsule
   center=<11.4034 -3.9702 -0.1658>
   rotation=<rot -0.0484 95.5306 -87.4947>
  }]
 kin_backSword [{shapeType=Capsule, length=40.7617, radius=5.0548, shapeName=kin_backSword_capsule, center=<2.3023 -2.346 -2.0445>, rotation=<rot 0 0 90>}]
 spine_01 [{shapeType=Capsule, length=8.5811, radius=10.1828, shapeName=spine_01_capsule, center=<0 -1.3857 0.2493>}]
 thigh_twist_01_l [{
   shapeType=Capsule
   length=19.5288
   radius=9.8759
   shapeName=thigh_twist_01_l_capsule
   center=<12.3827 -0.7124 0.8187>
   rotation=<rot 87.4428 174.9183 0>
  }]
 thigh_twist_01_r [{
   shapeType=Capsule
   length=19.1975
   radius=10.2073
   shapeName=thigh_twist_01_r_capsule
   center=<-12.3627 0.6638 -1.3601>
   rotation=<rot 87.4428 174.9182 -177.0827>
  }]
}
constraints [
 {bone1Name=dyn_potion, bone2Name=base_potion, swing1LimitDegrees=40, swing2LimitDegrees=40, twistLimitDegrees=0}
 {bone1Name=dyn_sideSword, bone2Name=base_sideSword, swing1LimitDegrees=30, swing2LimitDegrees=30, twistLimitDegrees=0}
 {
  bone1Name=base_potion
  bone2Name=spine_01
  swing1LimitDegrees=45
  swing1Motion=Free
  swing2LimitDegrees=45
  swing2Motion=Free
  twistLimitDegrees=45
  twistMotion=Free
 }
 {
  bone1Name=base_sideSword
  bone2Name=spine_01
  swing1LimitDegrees=45
  swing1Motion=Free
  swing2LimitDegrees=45
  swing2Motion=Free
  twistLimitDegrees=45
  twistMotion=Free
 }
 {bone1Name=dyn_sideBag, bone2Name=base_sideBag, swing1LimitDegrees=30, swing2LimitDegrees=25, twistLimitDegrees=0}
 {
  bone1Name=base_sideBag
  bone2Name=spine_01
  swing1LimitDegrees=45
  swing1Motion=Free
  swing2LimitDegrees=45
  swing2Motion=Free
  twistLimitDegrees=45
  twistMotion=Free
 }
 {
  bone1Name=kin_backSword
  bone2Name=spine_01
  swing1LimitDegrees=45
  swing1Motion=Free
  swing2LimitDegrees=45
  swing2Motion=Free
  twistLimitDegrees=45
  twistMotion=Free
 }
 {
  bone1Name=dyn_backSwordHangBelt
  bone2Name=kin_backSword
  swing1LimitDegrees=45
  swing1Motion=Free
  swing2LimitDegrees=45
  swing2Motion=Free
  twistLimitDegrees=45
  twistMotion=Free
 }
 {bone1Name=dyn_hipChain_2, bone2Name=dyn_hipChain_1, swing1LimitDegrees=20, swing2LimitDegrees=10, twistLimitDegrees=10}
 {bone1Name=dyn_hipChain_1, bone2Name=base_hipChain_1, swing1LimitDegrees=20, swing2LimitDegrees=10, twistLimitDegrees=10}
]
