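# Physics asset for the skirt belt of the DreamPaneer medium male player cosmetic, setting up physics for a short hip chain.
# Two capsule links, dyn_hipChain_1 and dyn_hipChain_2, hang in series from a tiny base_hipChain_1 anchor capsule through constraints with tight swing and twist limits.
# Two larger unconstrained capsules on the center and left coll_skirt_02 bones serve, by their naming, as skirt colliders for the chain.
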PhysicsAsset /BRCosmetics/Characters/Player/Male/Medium/Bodies/M_MED_DreamPaneer/Meshes/M_MED_DreamPaneer_SkirtBelt_Physics
bodies {
 base_hipChain_1 [{shapeType=Capsule, length=0.5, radius=0.25, shapeName=base_sideBag_capsule}]
 coll_skirt_02_c [{shapeType=Capsule, length=13.6208, radius=5.4398, shapeName=coll_skirt_02_r_capsule, center=<7.883 0.2833 -4.5403>, rotation=<rot 90 0 0>}]
 coll_skirt_02_l [{shapeType=Capsule, length=14.8129, radius=5.374, shapeName=coll_skirt_02_r_capsule, center=<7.8794 -0.2982 -3.6404>, rotation=<rot 90 0 0>}]
 dyn_hipChain_1 [{shapeType=Capsule, length=5.4383, radius=1.6786, shapeName=dyn_hipChain_1_capsule, center=<3.4727 0 0.571>, rotation=<rot -90 0 0>}]
 dyn_hipChain_2 [{shapeType=Capsule, length=5.4844, radius=1.6175, shapeName=dyn_hipChain_2_capsule, center=<3.6413 0 0.6193>, rotation=<rot -90 0 0>}]
}
constraints [
 {bone1Name=dyn_hipChain_2, bone2Name=dyn_hipChain_1, swing1LimitDegrees=20, swing2LimitDegrees=10, twistLimitDegrees=10}
 {bone1Name=dyn_hipChain_1, bone2Name=base_hipChain_1, swing1LimitDegrees=20, swing2LimitDegrees=10, twistLimitDegrees=10}
]
