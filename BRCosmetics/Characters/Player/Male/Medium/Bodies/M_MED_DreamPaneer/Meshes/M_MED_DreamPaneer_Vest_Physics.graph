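# Physics asset for the DreamPaneer male medium body's vest, setting up secondary motion for seven jacket flaps.
# Two spine capsules represent the torso, with a tiny anchor capsule per flap base joined to spine_03 by free angular constraints.
# Each larger dynamic flap capsule hangs off its base with tight swing limits and zero twist limit, so the flaps can sway only slightly.
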PhysicsAsset /BRCosmetics/Characters/Player/Male/Medium/Bodies/M_MED_DreamPaneer/Meshes/M_MED_DreamPaneer_Vest_Physics
bodies {
 base_jacketFlap_1_c [{shapeType=Capsule, length=0.505, radius=0.505, shapeName=base_jacketFlap_1_c_capsule, rotation=<rot 0 -90 0>}]
 base_jacketFlap_1_l [{shapeType=Capsule, length=0.505, radius=0.505, shapeName=base_jacketFlap_1_l_capsule, rotation=<rot 0 -90 0>}]
 base_jacketFlap_1_r [{shapeType=Capsule, length=0.505, radius=0.505, shapeName=base_jacketFlap_1_r_capsule, rotation=<rot 0 -90 0>}]
 base_jacketFlap_2_l [{shapeType=Capsule, length=0.505, radius=0.505, shapeName=base_jacketFlap_2_l_capsule, rotation=<rot 0 -90 0>}]
 base_jacketFlap_2_r [{shapeType=Capsule, length=0.505, radius=0.505, shapeName=base_jacketFlap_2_r_capsule, rotation=<rot 0 -90 0>}]
 base_jacketFlap_3_l [{shapeType=Capsule, length=0.505, radius=0.505, shapeName=base_jacketFlap_3_l_capsule, rotation=<rot 0 -90 0>}]
 base_jacketFlap_3_r [{shapeType=Capsule, length=0.505, radius=0.505, shapeName=base_jacketFlap_3_r_capsule, rotation=<rot 0 -90 0>}]
 dyn_jacketFlap_1_c [{shapeType=Capsule, length=11.2304, radius=3.2997, shapeName=dyn_jacketFlap_1_c_capsule, center=<7.3149 0 0.7854>, rotation=<rot 0 0 90>}]
 dyn_jacketFlap_1_l [{
   shapeType=Capsule
   length=7.2449
   radius=3.2997
   shapeName=dyn_jacketFlap_1_l_capsule
   center=<7.3149 -3.7997 1.3684>
   rotation=<rot 0 0 79.5898>
  }]
 dyn_jacketFlap_1_r [{
   shapeType=Capsule
   length=7.2449
   radius=3.2997
   shapeName=dyn_jacketFlap_1_r_capsule
   center=<7.3149 3.7997 1.3684>
   rotation=<rot 0 0 -100.4102>
  }]
 dyn_jacketFlap_2_l [{shapeType=Capsule, length=11.2304, radius=3.2997, shapeName=dyn_jacketFlap_2_l_capsule, center=<7.3149 0 2.263>, rotation=<rot 0 0 90>}]
 dyn_jacketFlap_2_r [{shapeType=Capsule, length=11.2304, radius=3.2997, shapeName=dyn_jacketFlap_2_r_capsule, center=<7.3149 0 2.263>, rotation=<rot 0 0 90>}]
 dyn_jacketFlap_3_l [{shapeType=Capsule, length=11.2304, radius=3.2997, shapeName=dyn_jacketFlap_3_l_capsule, center=<7.3149 0 0.7314>, rotation=<rot 0 0 90>}]
 dyn_jacketFlap_3_r [{shapeType=Capsule, length=11.2304, radius=3.2997, shapeName=dyn_jacketFlap_3_r_capsule, center=<7.3149 0 0.5479>, rotation=<rot 0 0 90>}]
 spine_02 [{shapeType=Capsule, length=9.0603, radius=8.7486, shapeName=spine_01_capsule, center=<0.3974 0 0>, rotation=<rot 0 -90 0>}]
 spine_03 [{shapeType=Capsule, length=8.1154, radius=9.3903, shapeName=spine_01_capsule, rotation=<rot 0 -90 0>}]
}
constraints [
 {bone1Name=dyn_jacketFlap_1_c, bone2Name=base_jacketFlap_1_c, swing1LimitDegrees=15, swing2LimitDegrees=20, twistLimitDegrees=0}
 {bone1Name=dyn_jacketFlap_1_l, bone2Name=base_jacketFlap_1_l, swing1LimitDegrees=15, swing2LimitDegrees=20, twistLimitDegrees=0}
 {bone1Name=dyn_jacketFlap_1_r, bone2Name=base_jacketFlap_1_r, swing1LimitDegrees=15, swing2LimitDegrees=20, twistLimitDegrees=0}
 {bone1Name=dyn_jacketFlap_2_l, bone2Name=base_jacketFlap_2_l, swing1LimitDegrees=15, swing2LimitDegrees=20, twistLimitDegrees=0}
 {bone1Name=dyn_jacketFlap_2_r, bone2Name=base_jacketFlap_2_r, swing1LimitDegrees=15, swing2LimitDegrees=20, twistLimitDegrees=0}
 {bone1Name=dyn_jacketFlap_3_l, bone2Name=base_jacketFlap_3_l, swing1LimitDegrees=15, swing2LimitDegrees=20, twistLimitDegrees=0}
 {bone1Name=dyn_jacketFlap_3_r, bone2Name=base_jacketFlap_3_r, swing1LimitDegrees=15, swing2LimitDegrees=20, twistLimitDegrees=0}
 {
  bone1Name=base_jacketFlap_1_c
  bone2Name=spine_03
  swing1LimitDegrees=45
  swing1Motion=Free
  swing2LimitDegrees=45
  swing2Motion=Free
  twistLimitDegrees=45
  twistMotion=Free
 }
 {
  bone1Name=base_jacketFlap_1_l
  bone2Name=spine_03
  swing1LimitDegrees=45
  swing1Motion=Free
  swing2LimitDegrees=45
  swing2Motion=Free
  twistLimitDegrees=45
  twistMotion=Free
 }
 {
  bone1Name=base_jacketFlap_2_l
  bone2Name=spine_03
  swing1LimitDegrees=45
  swing1Motion=Free
  swing2LimitDegrees=45
  swing2Motion=Free
  twistLimitDegrees=45
  twistMotion=Free
 }
 {
  bone1Name=base_jacketFlap_3_l
  bone2Name=spine_03
  swing1LimitDegrees=45
  swing1Motion=Free
  swing2LimitDegrees=45
  swing2Motion=Free
  twistLimitDegrees=45
  twistMotion=Free
 }
 {
  bone1Name=base_jacketFlap_1_r
  bone2Name=spine_03
  swing1LimitDegrees=45
  swing1Motion=Free
  swing2LimitDegrees=45
  swing2Motion=Free
  twistLimitDegrees=45
  twistMotion=Free
 }
 {
  bone1Name=base_jacketFlap_2_r
  bone2Name=spine_03
  swing1LimitDegrees=45
  swing1Motion=Free
  swing2LimitDegrees=45
  swing2Motion=Free
  twistLimitDegrees=45
  twistMotion=Free
 }
 {
  bone1Name=base_jacketFlap_3_r
  bone2Name=spine_03
  swing1LimitDegrees=45
  swing1Motion=Free
  swing2LimitDegrees=45
  swing2Motion=Free
  twistLimitDegrees=45
  twistMotion=Free
 }
 {
  bone1Name=spine_03
  bone2Name=spine_02
  swing1LimitDegrees=45
  swing1Motion=Free
  swing2LimitDegrees=45
  swing2Motion=Free
  twistLimitDegrees=45
  twistMotion=Free
 }
]
