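# Animation Blueprint for the DreamPaneer male body cosmetic, adding secondary motion to clothing and accessories.
# It copies the parent mesh pose through a Control Rig, then chains cached stages: bone adjustments, AnimDynamics, necklace setup and RigidBody simulations.
# Weighted constraints and LookAts aim skirt collision bones at simulated skirt bones.
# Custom idle state switches necklace dynamics and blends slot animation onto side-sword bones over the final pose.
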AnimBlueprint /BRCosmetics/Characters/Player/Male/Medium/Bodies/M_MED_DreamPaneer/Meshes/M_MED_DreamPaneer_AnimBP
target_skeleton=&/M_MED_DreamPaneer_Skeleton
graphs [
 AnimGraph {nodes [
   Output Pose(Result=Layered blend per bone(
     BasePose=Use cached pose 'postSimFixes'()
     BlendPoses_0=Slot 'DefaultSlot'()
     BlendWeights_0=Get Custom Idle is Playing Float() -> (Custom Idle is Playing Float:double)
    ) {layerSetup [{branchFilters [{boneName=base_sideSword}, {boneName=dyn_sideSword}]}]}) {layerGroup=None}
   n1 = Control Rig!(
    Source=Copy Pose From Mesh() {bUseAttachedParent=true}
    Alpha
    isSkyDivingFloat:bool=Get bIsSkydiveFloating()
    lowerLOD:bool=Get lowerLOD() -> (lowerLOD:bool)
   ) -> (Pose) {
    alphaBoolBlend {blendOption=Linear}
    controlRigAssetReference {controlRigAsset=&/M_MED_DreamPaneer_CtrlRig}
   }
   Save cached pose 'controlRigCP'(Pose=@n1) {cacheName=controlRigCP}
   n2 = "Constraint - coll_skirt_01_l (PPP)"!(
    ComponentPose=Local To Component(LocalPose=Use cached pose 'necklaceSetupCP'())
    Alpha=1
    ConstraintWeights_0=0.6081
    ConstraintWeights_1:float=0.3118
    ConstraintWeights_2:float=0.0801
   ) -> (Pose) {
    boneToModify {boneName=coll_skirt_01_l}
    constraintSetup [{targetBone {boneName=pelvis}}, {targetBone {boneName=deform_glute_l}}, {targetBone {boneName=thigh_twist_01_l}}]
    constraintWeights [1,1,1]
   }
   n3 = "Constraint - coll_skirt_01_c (PPPPP)"!(
    ComponentPose=@n2
    Alpha=1
    ConstraintWeights_0=0.74
    ConstraintWeights_1:float=0.1026
    ConstraintWeights_2:float=0.0684
    ConstraintWeights_3:float=0.0465
    ConstraintWeights_4:float=0.0424
   ) -> (Pose) {
    boneToModify {boneName=coll_skirt_01_c}
    constraintSetup [
     {targetBone {boneName=pelvis}}
     {targetBone {boneName=deform_glute_l}}
     {targetBone {boneName=deform_glute_r}}
     {targetBone {boneName=thigh_twist_01_l}}
     {targetBone {boneName=thigh_twist_01_r}}
    ]
    constraintWeights [1,1,1,1,1]
   }
   n4 = "Constraint - coll_skirt_01_r (PPP)"!(ComponentPose=@n3, Alpha=1, ConstraintWeights_0=0.7098, ConstraintWeights_1:float=0.2013, ConstraintWeights_2:float=0.0889) -> (Pose) {
    boneToModify {boneName=coll_skirt_01_r}
    constraintSetup [{targetBone {boneName=pelvis}}, {targetBone {boneName=deform_glute_r}}, {targetBone {boneName=thigh_twist_01_r}}]
    constraintWeights [1,1,1]
   }
   n5 = "Constraint - coll_skirt_02_r (PPP)"!(ComponentPose=@n4, Alpha=1, ConstraintWeights_0=0.6571, ConstraintWeights_1:float=0.207, ConstraintWeights_2:float=0.1359) -> (Pose) {
    boneToModify {boneName=coll_skirt_02_r}
    constraintSetup [{targetBone {boneName=pelvis}}, {targetBone {boneName=deform_glute_r}}, {targetBone {boneName=thigh_twist_01_r}}]
    constraintWeights [1,1,1]
   }
   n6 = "Constraint - coll_skirt_03_r (PPPP)"!(
    ComponentPose=@n5
    Alpha=1
    ConstraintWeights_0=0.6356
    ConstraintWeights_1:float=0.2512
    ConstraintWeights_2:float=0.0643
    ConstraintWeights_3:float=0.0489
   ) -> (Pose) {
    boneToModify {boneName=coll_skirt_03_r}
    constraintSetup [
     {targetBone {boneName=pelvis}}
     {targetBone {boneName=thigh_twist_01_r}}
     {targetBone {boneName=deform_glute_r}}
     {targetBone {boneName=deform_groin_r}}
    ]
    constraintWeights [1,1,1,1]
   }
   n7 = "Constraint - target_skirt_01_l (PPP)"!(ComponentPose=@n6, Alpha=1, ConstraintWeights_0=0.6994, ConstraintWeights_1:float=0.2556, ConstraintWeights_2:float=0.045) -> (Pose) {
    boneToModify {boneName=target_skirt_01_l}
    constraintSetup [{targetBone {boneName=dyn_skirt_3_l}}, {targetBone {boneName=dyn_skirt_2_c}}, {targetBone {boneName=dyn_skirt_2_l}}]
    constraintWeights [1,1,1]
   }
   n8 = "Constraint - target_skirt_01_c (PPP)"!(ComponentPose=@n7, Alpha=1, ConstraintWeights_0=0.6189, ConstraintWeights_1:float=0.1948, ConstraintWeights_2:float=0.1863) -> (Pose) {
    boneToModify {boneName=target_skirt_01_c}
    constraintSetup [{targetBone {boneName=dyn_skirt_2_c}}, {targetBone {boneName=dyn_skirt_3_r}}, {targetBone {boneName=dyn_skirt_3_l}}]
    constraintWeights [1,1,1]
   }
   n9 = "Constraint - target_skirt_01_r (PPPP)"!(
    ComponentPose=@n8
    Alpha=1
    ConstraintWeights_0=0.593
    ConstraintWeights_1:float=0.3671
    ConstraintWeights_2:float=0.0205
    ConstraintWeights_3:float=0.0194
   ) -> (Pose) {
    boneToModify {boneName=target_skirt_01_r}
    constraintSetup [
     {targetBone {boneName=dyn_skirt_3_r}}
     {targetBone {boneName=dyn_skirt_2_c}}
     {targetBone {boneName=dyn_skirt_2_r}}
     {targetBone {boneName=dyn_skirt_3_l}}
    ]
    constraintWeights [1,1,1,1]
   }
   n10 = "Constraint - target_skirt_02_r (PPP)"!(ComponentPose=@n9, Alpha=1, ConstraintWeights_0=0.6765, ConstraintWeights_1:float=0.3005, ConstraintWeights_2:float=0.0231) -> (Pose) {
    boneToModify {boneName=target_skirt_02_r}
    constraintSetup [{targetBone {boneName=dyn_skirt_3_r}}, {targetBone {boneName=dyn_skirt_2_r}}, {targetBone {boneName=dyn_skirt_2_c}}]
    constraintWeights [1,1,1]
   }
   n11 = "Constraint - target_skirt_03_r (PPP)"!(ComponentPose=@n10, Alpha=1, ConstraintWeights_0=0.7441, ConstraintWeights_1:float=0.1837, ConstraintWeights_2:float=0.0722) -> (Pose) {
    boneToModify {boneName=target_skirt_03_r}
    constraintSetup [{targetBone {boneName=dyn_skirt_2_r}}, {targetBone {boneName=dyn_skirt_1_r}}, {targetBone {boneName=dyn_skirt_3_r}}]
    constraintWeights [1,1,1]
   }
   n12 = "Constraint - coll_skirt_02_l (PPPPPP)"!(
    ComponentPose=@n11
    Alpha=1
    ConstraintWeights_0=0.3574
    ConstraintWeights_1:float=0.2587
    ConstraintWeights_2:float=0.2553
    ConstraintWeights_3:float=0.1029
    ConstraintWeights_4:float=0.0149
    ConstraintWeights_5:float=0.0108
   ) -> (Pose) {
    boneToModify {boneName=coll_skirt_02_l}
    constraintSetup [
     {targetBone {boneName=pelvis}}
     {targetBone {boneName=thigh_twist_01_l}}
     {targetBone {boneName=deform_groin_l}}
     {targetBone {boneName=dyn_skirt_1_l}}
     {targetBone {boneName=thigh_twist_01_r}}
     {targetBone {boneName=thigh_l}}
    ]
    constraintWeights [1,1,1,1,1,1]
   }
   n13 = "Constraint - coll_skirt_02_c (PPPPP)"!(
    ComponentPose=@n12
    Alpha=1
    ConstraintWeights_0=0.4651
    ConstraintWeights_1:float=0.197
    ConstraintWeights_2:float=0.1857
    ConstraintWeights_3:float=0.1019
    ConstraintWeights_4:float=0.0503
   ) -> (Pose) {
    boneToModify {boneName=coll_skirt_02_c}
    constraintSetup [
     {targetBone {boneName=pelvis}}
     {targetBone {boneName=deform_groin_l}}
     {targetBone {boneName=thigh_twist_01_l}}
     {targetBone {boneName=dyn_skirt_1_l}}
     {targetBone {boneName=thigh_twist_01_r}}
    ]
    constraintWeights [1,1,1,1,1]
   }
   n14 = "Constraint - target_skirt_02_c (PPPPPPPP)"!(
    ComponentPose=@n13
    Alpha=1
    ConstraintWeights_0=0.4348
    ConstraintWeights_1:float=0.1797
    ConstraintWeights_2:float=0.0986
    ConstraintWeights_3:float=0.0865
    ConstraintWeights_4:float=0.0743
    ConstraintWeights_5:float=0.07
    ConstraintWeights_6:float=0.0291
    ConstraintWeights_7:float=0.0269
   ) -> (Pose) {
    boneToModify {boneName=target_skirt_02_c}
    constraintSetup [
     {targetBone {boneName=dyn_skirt_1_l}}
     {targetBone {boneName=thigh_twist_01_l}}
     {targetBone {boneName=pelvis}}
     {targetBone {boneName=deform_groin_l}}
     {targetBone {boneName=thigh_twist_01_r}}
     {targetBone {boneName=thigh_l}}
     {targetBone {boneName=dyn_skirt_1_r}}
     {targetBone {boneName=thigh_r}}
    ]
    constraintWeights [1,1,1,1,1,1,1,1]
   }
   n15 = "Constraint - target_skirt_02_l (P)"!(ComponentPose=@n14, Alpha=1, ConstraintWeights_0=1, ConstraintWeights_1:float=0) -> (Pose) {boneToModify {boneName=target_skirt_02_l}, constraintSetup [{targetBone {boneName=dyn_skirt_1_l}}], constraintWeights [1,0]}
   n16 = Look At - Bone: coll_skirt_01_l(ComponentPose=@n15, Alpha=1) {
    boneToModify {boneName=coll_skirt_01_l}
    lookAtLocation {x=0}
    lookAtTarget {boneReference {boneName=target_skirt_01_l}}
    lookAt_Axis {axis=<1 0>}
    lookUp_Axis {axis {x=0, z=1}}
   }
   n17 = Look At - Bone: coll_skirt_01_c(ComponentPose=@n16, Alpha=1) {
    boneToModify {boneName=coll_skirt_01_c}
    lookAtLocation {x=0}
    lookAtTarget {boneReference {boneName=target_skirt_01_c}}
    lookAt_Axis {axis=<1 0>}
    lookUp_Axis {axis {x=0, z=1}}
   }
   n18 = Look At - Bone: coll_skirt_01_r(ComponentPose=@n17, Alpha=1) {
    boneToModify {boneName=coll_skirt_01_r}
    lookAtLocation {x=0}
    lookAtTarget {boneReference {boneName=target_skirt_01_r}}
    lookAt_Axis {axis=<1 0>}
    lookUp_Axis {axis {x=0, z=1}}
   }
   n19 = Look At - Bone: coll_skirt_02_r(ComponentPose=@n18, Alpha=1) {
    boneToModify {boneName=coll_skirt_02_r}
    lookAtLocation {x=0}
    lookAtTarget {boneReference {boneName=target_skirt_02_r}}
    lookAt_Axis {axis=<1 0>}
    lookUp_Axis {axis {x=0, z=1}}
   }
   n20 = Look At - Bone: coll_skirt_03_r(ComponentPose=@n19, Alpha=1) {
    boneToModify {boneName=coll_skirt_03_r}
    lookAtLocation {x=0}
    lookAtTarget {boneReference {boneName=target_skirt_03_r}}
    lookAt_Axis {axis=<1 0>}
    lookUp_Axis {axis {x=0, z=1}}
   }
   n21 = Look At - Bone: coll_skirt_02_c(ComponentPose=@n20, Alpha=1) {
    boneToModify {boneName=coll_skirt_02_c}
    lookAtLocation {x=0}
    lookAtTarget {boneReference {boneName=target_skirt_02_c}}
    lookAt_Axis {axis=<1 0>}
    lookUp_Axis {axis {x=0, z=1}}
   }
   n22 = Look At - Bone: coll_skirt_02_l(ComponentPose=@n21, Alpha=1) {
    boneToModify {boneName=coll_skirt_02_l}
    lookAtLocation {x=0}
    lookAtTarget {boneReference {boneName=target_skirt_02_l}}
    lookAt_Axis {axis=<1 0>}
    lookUp_Axis {axis {x=0, z=1}}
   }
   n23 = RigidBody!(ComponentPose=@n22, Alpha, OverrideWorldGravity:Vector="0,0,0") -> (Pose) {
    bOverrideWorldGravity=true
    overridePhysicsAsset=&/M_MED_DreamPaneer_Vest_Physics
    simulationSpace=BaseBoneSpace
    alphaBoolBlend {blendOption=Linear}
    baseBoneRef {boneName=spine_03}
    componentAppliedLinearAccClamp=<3000 3000 3000>
    simSpaceSettings {
     maxAngularAcceleration=2
     maxAngularVelocity=5
     maxLinearAcceleration=600
     maxLinearVelocity=400
     velocityScaleZ=0.8
     worldAlpha=1
     externalLinearDragV=<0.5 0.5 0.5>
    }
   }
   n24 = RigidBody(ComponentPose=@n23, OverrideWorldGravity="0,0,0") {
    bOverrideWorldGravity=true
    overridePhysicsAsset=&/M_MED_DreamPaneer_BackSword_Physics
    simulationSpace=BaseBoneSpace
    alphaBoolBlend {blendOption=Linear}
    baseBoneRef {boneName=pelvis}
    componentAppliedLinearAccClamp=<3000 3000 3000>
    simSpaceSettings {maxAngularAcceleration=1, maxAngularVelocity=6, maxLinearAcceleration=800, maxLinearVelocity=400, velocityScaleZ=0.6, worldAlpha=1}
   }
   n25 = Anim Dynamics - Bone: dyn_backSwordHangBelt!(ComponentPose=@n24, Alpha=1) -> (Pose) {
    angularDampingOverride=0.99
    angularSpringConstant=4000
    bAngularSpring=true
    bOverrideAngularDamping=true
    bUseGravityOverride=true
    gravityScale=0.35
    simulationSpace=RootRelative
    alphaBoolBlend {blendOption=Linear}
    boundBone {boneName=dyn_backSwordHangBelt}
    componentAppliedLinearAccClamp=<4000 4000 4000>
    componentLinearAccScale=<1 1 1>
    componentLinearVelScale=<1 1 1>
    gravityOverride {z=-980}
    physicsBodyDefinitions [{
      boundBone {boneName=dyn_backSwordHangBelt}
      boxExtents=<30 30 30>
      constraintSetup {angularLimitsMax=<0 20 15>, angularLimitsMin=<0 0 -15>, angularTarget=<1 0 0>}
      localJointOffset=<10 0 0>
     }]
    simSpaceSettings {maxAngularAcceleration=1, maxAngularVelocity=8, simSpaceAngularAlpha=0.75}
   }
   n26 = RigidBody(ComponentPose=@n25, OverrideWorldGravity="0,0,0") {
    bOverrideWorldGravity=true
    overridePhysicsAsset=&/M_MED_DreamPaneer_SkirtAcc_Physics
    simulationSpace=BaseBoneSpace
    alphaBoolBlend {blendOption=Linear}
    baseBoneRef {boneName=pelvis}
    componentAppliedLinearAccClamp=<3000 3000 3000>
    simSpaceSettings {maxAngularAcceleration=1, maxAngularVelocity=6, maxLinearAcceleration=1200, maxLinearVelocity=400, velocityScaleZ=0.8, worldAlpha=1}
   }
   n27 = RigidBody(ComponentPose=@n26, OverrideWorldGravity="0,0,0") {
    bOverrideWorldGravity=true
    overridePhysicsAsset=&/M_MED_DreamPaneer_SkirtBelt_Physics
    simulationSpace=BaseBoneSpace
    alphaBoolBlend {blendOption=Linear}
    baseBoneRef {boneName=coll_skirt_02_c}
    componentAppliedLinearAccClamp=<3000 3000 3000>
    simSpaceSettings {maxAngularAcceleration=1, maxAngularVelocity=6, maxLinearAcceleration=1200, maxLinearVelocity=400, velocityScaleZ=0.8, worldAlpha=1}
   }
   Save cached pose 'dynamicsCP'(Pose=Component To Local(ComponentPose=@n27)) {cacheName=dynamicsCP}
   n28 = Anim Dynamics - Bone: dyn_collarFur_3_l!(ComponentPose=Local To Component(LocalPose=Use cached pose 'generalAdjustmentsCP'()), Alpha=0.6) -> (Pose) {
    angularDampingOverride=1
    angularSpringConstant=8000
    bAngularSpring=true
    bOverrideAngularDamping=true
    gravityScale=0
    simulationSpace=RootRelative
    boundBone {boneName=dyn_collarFur_3_l}
    componentAppliedLinearAccClamp=<3000 3000 3000>
    componentLinearAccScale=<0.4 0.4 0.4>
    componentLinearVelScale=<0.4 0.4 0.4>
    gravityOverride {z=-980}
    physicsBodyDefinitions [{
      boundBone {boneName=dyn_collarFur_3_l}
      boxExtents=<20 20 20>
      constraintSetup {angularLimitsMax=<0 12 12>, angularLimitsMin=<0 -12 -12>, angularTarget=<1 0 0>}
      localJointOffset=<2 0 0>
     }]
   }
   n29 = Anim Dynamics - Bone: dyn_collarFur_2_l!(ComponentPose=@n28, Alpha=0.6) -> (Pose) {
    angularDampingOverride=1
    angularSpringConstant=8000
    bAngularSpring=true
    bOverrideAngularDamping=true
    gravityScale=0
    simulationSpace=RootRelative
    boundBone {boneName=dyn_collarFur_2_l}
    componentAppliedLinearAccClamp=<3000 3000 3000>
    componentLinearAccScale=<0.4 0.4 0.4>
    componentLinearVelScale=<0.4 0.4 0.4>
    gravityOverride {z=-980}
    physicsBodyDefinitions [{
      boundBone {boneName=dyn_collarFur_2_l}
      boxExtents=<20 20 20>
      constraintSetup {angularLimitsMax=<0 12 12>, angularLimitsMin=<0 -12 -12>, angularTarget=<1 0 0>}
      localJointOffset=<2 0 0>
     }]
   }
   n30 = Anim Dynamics - Bone: dyn_collarFur_1_l!(ComponentPose=@n29, Alpha=0.6) -> (Pose) {
    angularDampingOverride=1
    angularSpringConstant=8000
    bAngularSpring=true
    bOverrideAngularDamping=true
    gravityScale=0
    simulationSpace=RootRelative
    boundBone {boneName=dyn_collarFur_1_l}
    componentAppliedLinearAccClamp=<3000 3000 3000>
    componentLinearAccScale=<0.4 0.4 0.4>
    componentLinearVelScale=<0.4 0.4 0.4>
    gravityOverride {z=-980}
    physicsBodyDefinitions [{
      boundBone {boneName=dyn_collarFur_1_l}
      boxExtents=<20 20 20>
      constraintSetup {angularLimitsMax=<0 12 12>, angularLimitsMin=<0 -12 -12>, angularTarget=<1 0 0>}
      localJointOffset=<2 0 0>
     }]
   }
   n31 = Anim Dynamics - Bone: dyn_collarFur_1_c!(ComponentPose=@n30, Alpha=0.6) -> (Pose) {
    angularDampingOverride=1
    angularSpringConstant=8000
    bAngularSpring=true
    bOverrideAngularDamping=true
    gravityScale=0
    simulationSpace=RootRelative
    boundBone {boneName=dyn_collarFur_1_c}
    componentAppliedLinearAccClamp=<3000 3000 3000>
    componentLinearAccScale=<0.4 0.4 0.4>
    componentLinearVelScale=<0.4 0.4 0.4>
    gravityOverride {z=-980}
    physicsBodyDefinitions [{
      boundBone {boneName=dyn_collarFur_1_c}
      boxExtents=<20 20 20>
      constraintSetup {angularLimitsMax=<0 12 12>, angularLimitsMin=<0 -12 -12>, angularTarget=<1 0 0>}
      localJointOffset=<2 0 0>
     }]
   }
   n32 = Anim Dynamics - Bone: dyn_collarFur_3_r!(ComponentPose=@n31, Alpha=0.6) -> (Pose) {
    angularDampingOverride=1
    angularSpringConstant=8000
    bAngularSpring=true
    bOverrideAngularDamping=true
    gravityScale=0
    simulationSpace=RootRelative
    boundBone {boneName=dyn_collarFur_3_r}
    componentAppliedLinearAccClamp=<3000 3000 3000>
    componentLinearAccScale=<0.4 0.4 0.4>
    componentLinearVelScale=<0.4 0.4 0.4>
    gravityOverride {z=-980}
    physicsBodyDefinitions [{
      boundBone {boneName=dyn_collarFur_3_r}
      boxExtents=<20 20 20>
      constraintSetup {angularLimitsMax=<0 12 12>, angularLimitsMin=<0 -12 -12>, angularTarget=<1 0 0>}
      localJointOffset=<2 0 0>
     }]
   }
   n33 = Anim Dynamics - Bone: dyn_collarFur_2_r!(ComponentPose=@n32, Alpha=0.6) -> (Pose) {
    angularDampingOverride=1
    angularSpringConstant=8000
    bAngularSpring=true
    bOverrideAngularDamping=true
    gravityScale=0
    simulationSpace=RootRelative
    boundBone {boneName=dyn_collarFur_2_r}
    componentAppliedLinearAccClamp=<3000 3000 3000>
    componentLinearAccScale=<0.4 0.4 0.4>
    componentLinearVelScale=<0.4 0.4 0.4>
    gravityOverride {z=-980}
    physicsBodyDefinitions [{
      boundBone {boneName=dyn_collarFur_2_r}
      boxExtents=<20 20 20>
      constraintSetup {angularLimitsMax=<0 12 12>, angularLimitsMin=<0 -12 -12>, angularTarget=<1 0 0>}
      localJointOffset=<2 0 0>
     }]
   }
   n34 = Anim Dynamics - Bone: dyn_collarFur_1_r!(ComponentPose=@n33, Alpha=0.6) -> (Pose) {
    angularDampingOverride=1
    angularSpringConstant=8000
    bAngularSpring=true
    bOverrideAngularDamping=true
    gravityScale=0
    simulationSpace=RootRelative
    boundBone {boneName=dyn_collarFur_1_r}
    componentAppliedLinearAccClamp=<3000 3000 3000>
    componentLinearAccScale=<0.4 0.4 0.4>
    componentLinearVelScale=<0.4 0.4 0.4>
    gravityOverride {z=-980}
    physicsBodyDefinitions [{
      boundBone {boneName=dyn_collarFur_1_r}
      boxExtents=<20 20 20>
      constraintSetup {angularLimitsMax=<0 12 12>, angularLimitsMin=<0 -12 -12>, angularTarget=<1 0 0>}
      localJointOffset=<2 0 0>
     }]
   }
   n35 = Anim Dynamics - Bone: dyn_calfLace_l!(
    ComponentPose=@n34
    Alpha=1
    GravityOverride:Vector="0,0,-980"
    SimSpaceSettings:Anim Phys Sim Space Settings Structure=(SimSpaceAngularAlpha=0,MaxAngularVelocity=10000,MaxAngularAcceleration=10000,ExternalAngularVelocity=(X=0,Y=0,Z=0))
   ) -> (Pose) {
    angularDampingOverride=1
    angularSpringConstant=2000
    bAngularSpring=true
    bOverrideAngularDamping=true
    bUseGravityOverride=true
    gravityScale=0.35
    simulationSpace=RootRelative
    boundBone {boneName=dyn_calfLace_l}
    componentAppliedLinearAccClamp=<3000 3000 3000>
    componentLinearAccScale=<1 1 1>
    componentLinearVelScale=<1 1 1>
    gravityOverride {z=-980}
    physicsBodyDefinitions [{
      boundBone {boneName=dyn_calfLace_l}
      boxExtents=<20 20 20>
      constraintSetup {angularLimitsMax=<0 10 20>, angularLimitsMin=<0 -30 -20>, angularTarget=<1 0 0>}
      localJointOffset=<4 0 0>
     }]
   }
   n36 = Anim Dynamics - Bone: dyn_calfLace_r!(
    ComponentPose=@n35
    Alpha=1
    GravityOverride:Vector="0,0,-980"
    SimSpaceSettings:Anim Phys Sim Space Settings Structure=(SimSpaceAngularAlpha=0,MaxAngularVelocity=10000,MaxAngularAcceleration=10000,ExternalAngularVelocity=(X=0,Y=0,Z=0))
   ) -> (Pose) {
    angularDampingOverride=1
    angularSpringConstant=2000
    bAngularSpring=true
    bOverrideAngularDamping=true
    bUseGravityOverride=true
    gravityScale=0.35
    simulationSpace=RootRelative
    boundBone {boneName=dyn_calfLace_r}
    componentAppliedLinearAccClamp=<3000 3000 3000>
    componentLinearAccScale=<1 1 1>
    componentLinearVelScale=<1 1 1>
    gravityOverride {z=-980}
    physicsBodyDefinitions [{
      boundBone {boneName=dyn_calfLace_r}
      boxExtents=<20 20 20>
      constraintSetup {angularLimitsMax=<0 10 20>, angularLimitsMin=<0 -30 -20>, angularTarget=<1 0 0>}
      localJointOffset=<4 0 0>
     }]
   }
   n37 = Anim Dynamics - Bone: dyn_hood!(ComponentPose=@n36, bAlphaBoolEnabled:bool=True, GravityOverride:Vector="0,0,-980") -> (Pose) {
    alphaInputType=Bool
    angularDampingOverride=1
    angularSpringConstant=5000
    bAngularSpring=true
    bOverrideAngularDamping=true
    bUseGravityOverride=true
    gravityScale=0.35
    simulationSpace=RootRelative
    alphaBoolBlend {blendOption=Linear}
    boundBone {boneName=dyn_hood}
    componentAppliedLinearAccClamp=<3000 3000 3000>
    componentLinearAccScale=<0.5 0.5 1>
    componentLinearVelScale=<0.5 0.5 1>
    gravityOverride {z=-980}
    physicsBodyDefinitions [{
      boundBone {boneName=dyn_hood}
      boxExtents=<20 20 20>
      constraintSetup {angularLimitsMax=<0 5 0>, angularLimitsMin=<0 -5 -45>, angularTarget=<1 0 0>}
      localJointOffset=<1.5 0 0>
     }]
   }
   n38 = Anim Dynamics - Bone: dyn_skirt_1_l!(
    ComponentPose=@n37
    Alpha=1
    GravityOverride:Vector="0,0,-980"
    SimSpaceSettings:Anim Phys Sim Space Settings Structure=(SimSpaceAngularAlpha=0,MaxAngularVelocity=10000,MaxAngularAcceleration=10000,ExternalAngularVelocity=(X=0,Y=0,Z=0))
   ) -> (Pose) {
    angularDampingOverride=0.99
    angularSpringConstant=1000
    bAngularSpring=true
    bOverrideAngularDamping=true
    bUseGravityOverride=true
    gravityScale=0.35
    simulationSpace=RootRelative
    alphaBoolBlend {blendOption=Linear}
    boundBone {boneName=dyn_skirt_1_l}
    componentAppliedLinearAccClamp=<1000 1000 1000>
    componentLinearAccScale=<0.7 0.7 0.7>
    componentLinearVelScale=<0.7 0.7 0.7>
    gravityOverride {z=-980}
    physicsBodyDefinitions [{
      boundBone {boneName=dyn_skirt_1_l}
      boxExtents=<12 12 12>
      constraintSetup {angularLimitsMax=<0 0 0>, angularLimitsMin=<0 -10 0>, angularTarget=<1 0 0>}
      localJointOffset=<5 0 0>
     }]
   }
   n39 = Anim Dynamics - Bone: dyn_skirt_2_l!(
    ComponentPose=@n38
    Alpha=1
    GravityOverride:Vector="0,0,-980"
    SimSpaceSettings:Anim Phys Sim Space Settings Structure=(SimSpaceAngularAlpha=0,MaxAngularVelocity=10000,MaxAngularAcceleration=10000,ExternalAngularVelocity=(X=0,Y=0,Z=0))
   ) -> (Pose) {
    angularDampingOverride=0.99
    angularSpringConstant=1000
    bAngularSpring=true
    bOverrideAngularDamping=true
    bUseGravityOverride=true
    gravityScale=0.35
    simulationSpace=RootRelative
    alphaBoolBlend {blendOption=Linear}
    boundBone {boneName=dyn_skirt_2_l}
    componentAppliedLinearAccClamp=<1000 1000 1000>
    componentLinearAccScale=<0.7 0.7 0.7>
    componentLinearVelScale=<0.7 0.7 0.7>
    gravityOverride {z=-980}
    physicsBodyDefinitions [{
      boundBone {boneName=dyn_skirt_2_l}
      boxExtents=<12 12 12>
      constraintSetup {angularLimitsMax=<0 0 0>, angularLimitsMin=<0 -10 0>, angularTarget=<1 0 0>}
      localJointOffset=<5 0 0>
     }]
   }
   n40 = Anim Dynamics - Bone: dyn_skirt_3_l!(
    ComponentPose=@n39
    Alpha=1
    GravityOverride:Vector="0,0,-980"
    SimSpaceSettings:Anim Phys Sim Space Settings Structure=(SimSpaceAngularAlpha=0,MaxAngularVelocity=10000,MaxAngularAcceleration=10000,ExternalAngularVelocity=(X=0,Y=0,Z=0))
   ) -> (Pose) {
    angularDampingOverride=0.99
    angularSpringConstant=1000
    bAngularSpring=true
    bOverrideAngularDamping=true
    bUseGravityOverride=true
    gravityScale=0.35
    simulationSpace=RootRelative
    alphaBoolBlend {blendOption=Linear}
    boundBone {boneName=dyn_skirt_3_l}
    componentAppliedLinearAccClamp=<1000 1000 1000>
    componentLinearAccScale=<0.7 0.7 0.7>
    componentLinearVelScale=<0.7 0.7 0.7>
    gravityOverride {z=-980}
    physicsBodyDefinitions [{
      boundBone {boneName=dyn_skirt_3_l}
      boxExtents=<12 12 12>
      constraintSetup {angularLimitsMax=<0 0 0>, angularLimitsMin=<0 -10 0>, angularTarget=<1 0 0>}
      localJointOffset=<5 0 0>
     }]
   }
   n41 = Anim Dynamics - Bone: dyn_skirt_1_r!(ComponentPose=@n40, Alpha=0, GravityOverride:Vector="0,0,-980") -> (Pose) {
    angularDampingOverride=0.99
    angularSpringConstant=1000
    bAngularSpring=true
    bOverrideAngularDamping=true
    bUseGravityOverride=true
    gravityScale=0.35
    simulationSpace=RootRelative
    alphaBoolBlend {blendOption=Linear}
    boundBone {boneName=dyn_skirt_1_r}
    componentAppliedLinearAccClamp=<1000 1000 1000>
    componentLinearAccScale=<0.7 0.7 0.7>
    componentLinearVelScale=<0.7 0.7 0.7>
    gravityOverride {z=-980}
    physicsBodyDefinitions [{
      boundBone {boneName=dyn_skirt_1_r}
      boxExtents=<12 12 12>
      constraintSetup {angularLimitsMax=<0 0 0>, angularLimitsMin=<0 -10 0>, angularTarget=<1 0 0>}
      localJointOffset=<5 0 0>
     }]
   }
   n42 = Anim Dynamics - Bone: dyn_skirt_2_r!(
    ComponentPose=@n41
    Alpha=1
    GravityOverride:Vector="0,0,-980"
    SimSpaceSettings:Anim Phys Sim Space Settings Structure=(SimSpaceAngularAlpha=0,MaxAngularVelocity=10000,MaxAngularAcceleration=10000,ExternalAngularVelocity=(X=0,Y=0,Z=0))
   ) -> (Pose) {
    angularDampingOverride=0.99
    angularSpringConstant=1000
    bAngularSpring=true
    bOverrideAngularDamping=true
    bUseGravityOverride=true
    gravityScale=0.35
    simulationSpace=RootRelative
    alphaBoolBlend {blendOption=Linear}
    boundBone {boneName=dyn_skirt_2_r}
    componentAppliedLinearAccClamp=<1000 1000 1000>
    componentLinearAccScale=<0.7 0.7 0.7>
    componentLinearVelScale=<0.7 0.7 0.7>
    gravityOverride {z=-980}
    physicsBodyDefinitions [{
      boundBone {boneName=dyn_skirt_2_r}
      boxExtents=<12 12 12>
      constraintSetup {angularLimitsMax=<0 0 0>, angularLimitsMin=<0 -10 0>, angularTarget=<1 0 0>}
      localJointOffset=<5 0 0>
     }]
   }
   n43 = Anim Dynamics - Bone: dyn_skirt_3_r!(
    ComponentPose=@n42
    Alpha=1
    GravityOverride:Vector="0,0,-980"
    SimSpaceSettings:Anim Phys Sim Space Settings Structure=(SimSpaceAngularAlpha=0,MaxAngularVelocity=10000,MaxAngularAcceleration=10000,ExternalAngularVelocity=(X=0,Y=0,Z=0))
   ) -> (Pose) {
    angularDampingOverride=0.99
    angularSpringConstant=1000
    bAngularSpring=true
    bOverrideAngularDamping=true
    bUseGravityOverride=true
    gravityScale=0.35
    simulationSpace=RootRelative
    alphaBoolBlend {blendOption=Linear}
    boundBone {boneName=dyn_skirt_3_r}
    componentAppliedLinearAccClamp=<1000 1000 1000>
    componentLinearAccScale=<0.7 0.7 0.7>
    componentLinearVelScale=<0.7 0.7 0.7>
    gravityOverride {z=-980}
    physicsBodyDefinitions [{
      boundBone {boneName=dyn_skirt_3_r}
      boxExtents=<12 12 12>
      constraintSetup {angularLimitsMax=<0 0 0>, angularLimitsMin=<0 -10 0>, angularTarget=<1 0 0>}
      localJointOffset=<5 0 0>
     }]
   }
   n44 = Anim Dynamics - Bone: dyn_skirt_2_c!(
    ComponentPose=@n43
    Alpha=1
    GravityOverride:Vector="0,0,-980"
    SimSpaceSettings:Anim Phys Sim Space Settings Structure=(SimSpaceAngularAlpha=0,MaxAngularVelocity=10000,MaxAngularAcceleration=10000,ExternalAngularVelocity=(X=0,Y=0,Z=0))
   ) -> (Pose) {
    angularDampingOverride=0.99
    angularSpringConstant=1000
    bAngularSpring=true
    bOverrideAngularDamping=true
    bUseGravityOverride=true
    gravityScale=0.35
    simulationSpace=RootRelative
    alphaBoolBlend {blendOption=Linear}
    boundBone {boneName=dyn_skirt_2_c}
    componentAppliedLinearAccClamp=<1000 1000 1000>
    componentLinearAccScale=<0.7 0.7 0.7>
    componentLinearVelScale=<0.7 0.7 0.7>
    gravityOverride {z=-980}
    physicsBodyDefinitions [{
      boundBone {boneName=dyn_skirt_2_c}
      boxExtents=<12 12 12>
      constraintSetup {angularLimitsMax=<0 10 0>, angularLimitsMin=<0 -20 0>, angularTarget=<1 0 0>}
      localJointOffset=<5 0 0>
     }]
   }
   n45 = Anim Dynamics - Bone: dyn_backBagBelt_l!(
    ComponentPose=@n44
    Alpha=1
    SimSpaceSettings:Anim Phys Sim Space Settings Structure=(SimSpaceAngularAlpha=0,MaxAngularVelocity=10000,MaxAngularAcceleration=10000,ExternalAngularVelocity=(X=0,Y=0,Z=0))
   ) -> (Pose) {
    angularDampingOverride=1
    angularSpringConstant=12000
    bAngularSpring=true
    bOverrideAngularDamping=true
    bUseGravityOverride=true
    gravityScale=0.35
    simulationSpace=RootRelative
    boundBone {boneName=dyn_backBagBelt_l}
    componentAppliedLinearAccClamp=<3000 3000 3000>
    componentLinearAccScale=<0.6 0.6 1>
    componentLinearVelScale=<0.6 0.6 1>
    gravityOverride {z=-980}
    physicsBodyDefinitions [{
      boundBone {boneName=dyn_backBagBelt_l}
      boxExtents=<30 30 30>
      constraintSetup {angularLimitsMax=<0 0 10>, angularLimitsMin=<0 -50 -10>, angularTarget=<1 0 0>}
      localJointOffset=<10 0 0>
     }]
   }
   n46 = Anim Dynamics - Bone: dyn_backBagBelt_r!(
    ComponentPose=@n45
    Alpha=1
    SimSpaceSettings:Anim Phys Sim Space Settings Structure=(SimSpaceAngularAlpha=0,MaxAngularVelocity=10000,MaxAngularAcceleration=10000,ExternalAngularVelocity=(X=0,Y=0,Z=0))
   ) -> (Pose) {
    angularDampingOverride=1
    angularSpringConstant=12000
    bAngularSpring=true
    bOverrideAngularDamping=true
    bUseGravityOverride=true
    gravityScale=0.35
    simulationSpace=RootRelative
    boundBone {boneName=dyn_backBagBelt_r}
    componentAppliedLinearAccClamp=<3000 3000 3000>
    componentLinearAccScale=<0.6 0.6 1>
    componentLinearVelScale=<0.6 0.6 1>
    gravityOverride {z=-980}
    physicsBodyDefinitions [{
      boundBone {boneName=dyn_backBagBelt_r}
      boxExtents=<30 30 30>
      constraintSetup {angularLimitsMax=<0 0 10>, angularLimitsMin=<0 -50 -10>, angularTarget=<1 0 0>}
      localJointOffset=<10 0 0>
     }]
   }
   Save cached pose 'animDynamicsCP'(Pose=Component To Local(ComponentPose=@n46)) {cacheName=animDynamicsCP}
   n47 = "Constraint - attach_cape (P)"(ComponentPose=Local To Component(LocalPose=Use cached pose 'controlRigCP'()), Alpha=1, ConstraintWeights_0=1) {
    alphaBoolBlend {blendOption=Linear}
    boneToModify {boneName=attach_cape}
    constraintSetup [{targetBone {boneName=spine_05}}]
    constraintWeights [1]
   }
   n48 = "Transform (Modify) Bone - Bone: attach_cape"!(ComponentPose=@n47, bAlphaBoolEnabled:bool=True, Translation="0,-1,0", Rotation="0,0,16") -> (Pose) {
    alphaInputType=Bool
    rotationMode=BMM_Additive
    rotationSpace=BCS_BoneSpace
    translationMode=BMM_Additive
    translationSpace=BCS_BoneSpace
    alphaBoolBlend {blendInTime=0.5, blendOption=CircularOut, blendOutTime=0.5}
    boneToModify {boneName=attach_cape}
    rotation {roll=16}
    translation {y=-1}
   }
   n49 = "Transform (Modify) Bone - Bone: attach_cape"(ComponentPose=@n48, bAlphaBoolEnabled=True, Translation="0,-2,0", Rotation="0,0,20") {
    alphaInputType=Bool
    rotationMode=BMM_Additive
    rotationSpace=BCS_BoneSpace
    translationMode=BMM_Additive
    translationSpace=BCS_BoneSpace
    alphaBoolBlend {blendInTime=0.5, blendOption=CircularOut, blendOutTime=0.5}
    boneToModify {boneName=attach_cape}
    rotation {roll=20}
    translation {y=-2}
   }
   n50 = "Transform (Modify) Bone - Bone: attach_cape"!(ComponentPose=@n49, Alpha:float=1, Translation="0,-2,0", Rotation="0,0,20") -> (Pose) {
    rotationMode=BMM_Additive
    rotationSpace=BCS_BoneSpace
    translationMode=BMM_Additive
    translationSpace=BCS_BoneSpace
    alphaBoolBlend {blendOption=CircularOut}
    boneToModify {boneName=attach_cape}
    rotation {roll=20}
    translation {y=-2}
   }
   n51 = "Transform (Modify) Bone - Bone: attach_cape"!(ComponentPose=@n50, bAlphaBoolEnabled:bool=True, Translation="0,0,0", Rotation="0,0,10") -> (Pose) {
    alphaInputType=Bool
    rotationMode=BMM_Additive
    rotationSpace=BCS_BoneSpace
    translationMode=BMM_Additive
    translationSpace=BCS_BoneSpace
    alphaBoolBlend {blendInTime=0.5, blendOption=Cubic, blendOutTime=0.5}
    boneToModify {boneName=attach_cape}
    rotation {roll=10}
   }
   n52 = "Transform (Modify) Bone - Bone: kin_backSword"!(ComponentPose=@n51, bAlphaBoolEnabled:bool=True, Rotation="-40,0,0") -> (Pose) {
    alphaInputType=Bool
    rotationMode=BMM_Additive
    rotationSpace=BCS_BoneSpace
    alphaBoolBlend {blendInTime=0.5, blendOption=Cubic, blendOutTime=0.5}
    boneToModify {boneName=kin_backSword}
    rotation {pitch=-40}
   }
   n53 = "Transform (Modify) Bone - Bone: base_backBag"!(ComponentPose=@n52, bAlphaBoolEnabled:bool=True, Rotation="-20,0,0") -> (Pose) {
    alphaInputType=Bool
    rotationMode=BMM_Additive
    rotationSpace=BCS_BoneSpace
    alphaBoolBlend {blendInTime=0.5, blendOption=Cubic, blendOutTime=0.5}
    boneToModify {boneName=base_backBag}
    rotation {pitch=-20}
   }
   n54 = "Transform (Modify) Bone - Bone: dyn_skirt_2_c"!(ComponentPose=@n53, bAlphaBoolEnabled:bool=True, Translation="-3,0,0", Scale="0.7,1,1") -> (Pose) {
    alphaInputType=Bool
    rotationMode=BMM_Additive
    rotationSpace=BCS_BoneSpace
    scaleMode=BMM_Additive
    scaleSpace=BCS_BoneSpace
    translationMode=BMM_Additive
    translationSpace=BCS_BoneSpace
    alphaBoolBlend {blendInTime=0.5, blendOption=Cubic, blendOutTime=0.35}
    boneToModify {boneName=dyn_skirt_2_c}
    scale {x=0.7}
    translation {x=-3}
   }
   n55 = "Transform (Modify) Bone - Bone: dyn_skirt_3_l"!(ComponentPose=@n54, bAlphaBoolEnabled:bool=True, Translation="-1.5,0,0", Scale="0.8,1,1") -> (Pose) {
    alphaInputType=Bool
    rotationMode=BMM_Additive
    rotationSpace=BCS_BoneSpace
    scaleMode=BMM_Additive
    scaleSpace=BCS_BoneSpace
    translationMode=BMM_Additive
    translationSpace=BCS_BoneSpace
    alphaBoolBlend {blendInTime=0.5, blendOption=Cubic, blendOutTime=0.35}
    boneToModify {boneName=dyn_skirt_3_l}
    scale {x=0.8}
    translation {x=-1.5}
   }
   n56 = "Transform (Modify) Bone - Bone: base_necklace_1"!(ComponentPose=@n55, bAlphaBoolEnabled:bool=True, Translation="0.2,0,0.5", Rotation="50,0,0") -> (Pose) {
    alphaInputType=Bool
    rotationMode=BMM_Additive
    rotationSpace=BCS_BoneSpace
    scaleMode=BMM_Additive
    scaleSpace=BCS_BoneSpace
    translationMode=BMM_Additive
    translationSpace=BCS_BoneSpace
    alphaBoolBlend {blendInTime=0.5, blendOption=Cubic, blendOutTime=0.5}
    boneToModify {boneName=base_necklace_1}
    rotation {pitch=50}
    translation {x=0.2, z=0.5}
   }
   n57 = "Transform (Modify) Bone - Bone: coll_neckLace_chest_l"!(ComponentPose=@n56, Alpha=1, Translation="0,0,0.7") -> (Pose) {translationMode=BMM_Additive, translationSpace=BCS_BoneSpace, boneToModify {boneName=coll_neckLace_chest_l}, translation {z=0.7}}
   Save cached pose 'generalAdjustmentsCP'(Pose=Component To Local(ComponentPose=@n57)) {cacheName=generalAdjustmentsCP}
   n58 = Anim Dynamics - Bone: dyn_necklace_1!(
    ComponentPose=Local To Component(LocalPose=Use cached pose 'animDynamicsCP'())
    Alpha=1
    GravityOverride:Vector="0,0,-980"
    SimSpaceSettings:Anim Phys Sim Space Settings Structure=(SimSpaceAngularAlpha=0,MaxAngularVelocity=10000,MaxAngularAcceleration=10000,ExternalAngularVelocity=(X=0,Y=0,Z=0))
   ) -> (Pose) {
    angularDampingOverride=1
    angularSpringConstant=1000
    bAngularSpring=true
    bOverrideAngularDamping=true
    bUseGravityOverride=true
    bUseSphericalLimits=true
    gravityScale=0.35
    simulationSpace=RootRelative
    boundBone {boneName=dyn_necklace_1}
    componentAppliedLinearAccClamp=<4000 4000 4000>
    componentLinearAccScale=<1 1 0.5>
    componentLinearVelScale=<1 1 0.5>
    gravityOverride {z=-980}
    physicsBodyDefinitions [{
      boundBone {boneName=dyn_necklace_1}
      boxExtents=<20 20 20>
      constraintSetup {angularLimitsMax=<0 0 3>, angularLimitsMin=<0 -50 -3>, angularTarget=<1 0 0>}
      localJointOffset=<4 0 0>
     }]
    sphericalLimits [
     {limitRadius=10, drivingBone {boneName=coll_neckLace_chest_l}, sphereLocalOffset=<2.7247 -0.9121 -7.5808>}
     {limitRadius=10, drivingBone {boneName=coll_neckLace_chest_r}, sphereLocalOffset=<2.7247 0.4232 -7.3039>}
     {limitRadius=14.2614, drivingBone {boneName=spine_05}, sphereLocalOffset=<-3.5128 -3.3038 0>}
    ]
   }
   n59 = Anim Dynamics - Bone: dyn_necklace_2!(
    ComponentPose=@n58
    Alpha=1
    GravityOverride:Vector="0,0,-980"
    SimSpaceSettings:Anim Phys Sim Space Settings Structure=(SimSpaceAngularAlpha=0,MaxAngularVelocity=10000,MaxAngularAcceleration=10000,ExternalAngularVelocity=(X=0,Y=0,Z=0))
   ) -> (Pose) {
    angularDampingOverride=1
    angularSpringConstant=1000
    bAngularSpring=true
    bOverrideAngularDamping=true
    bUseGravityOverride=true
    bUseSphericalLimits=true
    gravityScale=0.35
    simulationSpace=RootRelative
    boundBone {boneName=dyn_necklace_2}
    componentAppliedLinearAccClamp=<4000 4000 4000>
    componentLinearAccScale=<1 1 1>
    componentLinearVelScale=<1 1 1>
    gravityOverride {z=-980}
    physicsBodyDefinitions [{
      boundBone {boneName=dyn_necklace_2}
      boxExtents=<20 20 20>
      constraintSetup {angularLimitsMax=<0 0 3>, angularLimitsMin=<0 -25 -3>, angularTarget=<1 0 0>}
      localJointOffset=<6 0 0>
     }]
    sphericalLimits [
     {limitRadius=10, drivingBone {boneName=coll_neckLace_chest_l}, sphereLocalOffset=<5.2832 -0.9121 -7.1222>}
     {limitRadius=10, drivingBone {boneName=coll_neckLace_chest_r}, sphereLocalOffset=<5.2832 0.4232 -6.8454>}
     {limitRadius=14.2614, drivingBone {boneName=spine_04}, sphereLocalOffset=<-0.149 0 0>}
    ]
   }
   n60 = Component To Local(ComponentPose=@n59)
   n61 = Anim Dynamics - Bone: dyn_necklace_1!(ComponentPose=Local To Component(LocalPose=Use cached pose 'animDynamicsCP'()), Alpha=1, GravityOverride="0,0,-1280") -> (Pose) {
    angularDampingOverride=1
    angularSpringConstant=4000
    bAngularSpring=true
    bOverrideAngularDamping=true
    bUseGravityOverride=true
    bUseSphericalLimits=true
    gravityScale=0.35
    simulationSpace=RootRelative
    boundBone {boneName=dyn_necklace_1}
    componentAppliedLinearAccClamp=<4000 4000 4000>
    componentLinearAccScale=<1 1 1>
    componentLinearVelScale=<1 1 1>
    gravityOverride {z=-980}
    physicsBodyDefinitions [{
      boundBone {boneName=dyn_necklace_1}
      boxExtents=<30 30 30>
      constraintSetup {angularLimitsMax=<0 0 3>, angularLimitsMin=<0 -50 -3>, angularTarget=<1 0 0>}
      localJointOffset=<15 0 0>
     }]
    simSpaceSettings {maxAngularAcceleration=4, maxAngularVelocity=12, simSpaceAngularAlpha=1}
    sphericalLimits [
     {limitRadius=10, drivingBone {boneName=coll_neckLace_chest_l}, sphereLocalOffset=<2.7247 -0.9121 -7.5808>}
     {limitRadius=10, drivingBone {boneName=coll_neckLace_chest_r}, sphereLocalOffset=<2.7247 0.4232 -7.3039>}
     {limitRadius=14.2614, drivingBone {boneName=spine_05}, sphereLocalOffset=<-3.5128 -3.3038 0>}
    ]
   }
   n62 = Anim Dynamics - Bone: dyn_necklace_2!(ComponentPose=@n61, Alpha=1, GravityOverride="0,0,-1280") -> (Pose) {
    angularDampingOverride=1
    angularSpringConstant=4000
    bAngularSpring=true
    bOverrideAngularDamping=true
    bUseGravityOverride=true
    bUseSphericalLimits=true
    gravityScale=0.35
    simulationSpace=RootRelative
    boundBone {boneName=dyn_necklace_2}
    componentAppliedLinearAccClamp=<4000 4000 4000>
    componentLinearAccScale=<1 1 1>
    componentLinearVelScale=<1 1 1>
    gravityOverride {z=-980}
    physicsBodyDefinitions [{
      boundBone {boneName=dyn_necklace_2}
      boxExtents=<30 30 30>
      constraintSetup {angularLimitsMax=<0 0 8>, angularLimitsMin=<0 -25 -8>, angularTarget=<1 0 0>}
      localJointOffset=<15 0 0>
     }]
    simSpaceSettings {maxAngularAcceleration=4, maxAngularVelocity=12, simSpaceAngularAlpha=1}
    sphericalLimits [
     {limitRadius=10, drivingBone {boneName=coll_neckLace_chest_l}, sphereLocalOffset=<5.2832 -0.9121 -7.1222>}
     {limitRadius=10, drivingBone {boneName=coll_neckLace_chest_r}, sphereLocalOffset=<5.2832 0.4232 -6.8454>}
     {limitRadius=14.2614, drivingBone {boneName=spine_04}, sphereLocalOffset=<-0.149 0 0>}
    ]
   }
   n63 = Component To Local(ComponentPose=@n62)
   n64 = Get Not Custom Idle is Playing bool() -> (Not Custom Idle is Playing bool:bool)
   n65 = "Constraint - kin_lowerNecklace_l (PPPP)"!(
    ComponentPose=Local To Component(LocalPose=Blend Poses by bool(BlendPose_0=@n60, BlendPose_1=@n63, bActiveValue=@n64))
    Alpha=1
    ConstraintWeights_0=0.7235
    ConstraintWeights_1:float=0.2006
    ConstraintWeights_2:float=0.0571
    ConstraintWeights_3:float=0.0188
   ) -> (Pose) {
    boneToModify {boneName=kin_lowerNecklace_l}
    constraintSetup [{targetBone {boneName=spine_05}}, {targetBone {boneName=clavicle_l}}, {targetBone {boneName=spine_04}}, {targetBone {boneName=spine_03}}]
    constraintWeights [1,1,1,1]
   }
   n66 = "Constraint - kin_lowerNecklace_r (PPPP)"!(
    ComponentPose=@n65
    Alpha=1
    ConstraintWeights_0=0.7233
    ConstraintWeights_1:float=0.2007
    ConstraintWeights_2:float=0.0573
    ConstraintWeights_3:float=0.0188
   ) -> (Pose) {
    boneToModify {boneName=kin_lowerNecklace_r}
    constraintSetup [{targetBone {boneName=spine_05}}, {targetBone {boneName=clavicle_r}}, {targetBone {boneName=spine_04}}, {targetBone {boneName=spine_03}}]
    constraintWeights [1,1,1,1]
   }
   n67 = Look At - Bone: kin_lowerNecklace_l(ComponentPose=@n66, Alpha=1) {
    boneToModify {boneName=kin_lowerNecklace_l}
    lookAtLocation {x=0}
    lookAtTarget {boneReference {boneName=dyn_necklace_2}}
    lookAt_Axis {axis=<1 0>}
    lookUp_Axis {axis {x=0, z=1}}
   }
   n68 = Look At - Bone: kin_lowerNecklace_r(ComponentPose=@n67, Alpha=1) {
    boneToModify {boneName=kin_lowerNecklace_r}
    lookAtLocation {x=0}
    lookAtTarget {boneReference {boneName=dyn_necklace_2}}
    lookAt_Axis {axis=<1 0>}
    lookUp_Axis {axis {x=0, z=1}}
   }
   Save cached pose 'necklaceSetupCP'(Pose=Component To Local(ComponentPose=@n68)) {cacheName=necklaceSetupCP}
   n69 = "Transform (Modify) Bone - Bone: dyn_backSwordHangBelt"!(ComponentPose=Local To Component(LocalPose=Use cached pose 'dynamicsCP'()), bAlphaBoolEnabled:bool=True, Rotation="-60,0,0") -> (Pose) {
    alphaInputType=Bool
    rotationMode=BMM_Additive
    rotationSpace=BCS_BoneSpace
    alphaBoolBlend {blendInTime=0.5, blendOption=Cubic, blendOutTime=0.5}
    boneToModify {boneName=dyn_backSwordHangBelt}
    rotation {pitch=-60}
   }
   n70 = "Transform (Modify) Bone - Bone: dyn_necklace_2"!(ComponentPose=@n69, bAlphaBoolEnabled:bool=True, Rotation="25,0,0") -> (Pose) {
    alphaInputType=Bool
    rotationMode=BMM_Additive
    rotationSpace=BCS_BoneSpace
    scaleMode=BMM_Additive
    scaleSpace=BCS_BoneSpace
    translationMode=BMM_Additive
    translationSpace=BCS_BoneSpace
    alphaBoolBlend {blendInTime=0.5, blendOption=Cubic, blendOutTime=0.5}
    boneToModify {boneName=dyn_necklace_2}
    rotation {pitch=25}
   }
   Save cached pose 'postSimFixes'(Pose=Component To Local(ComponentPose=@n70)) {cacheName=postSimFixes}
  ]}
 EventGraph {nodes [
   Event BlueprintUpdateAnimation()
   > n1 = Sequence!(execute) -> (then_0, then_1, then_2:Exec)
   > n2 = "Gravity Override State Function (Parameter Asset) (Nativized)"(
    CharacterPartAnimInstance=Self-Reference()
    GameTimeInSeconds=GetGameTimeInSeconds()
    Parameters=&/M_MED_DreamPaneer_Parameters
   )
   > Set GravityOverride(GravityOverride=@n2)
   n3 = Get bIsTacticalSprinting()
   n4 = Get bIsFalling()
   n5 = Get bIsJumping()
   n6 = Branch(
    execute=@n1.then_1
    Condition=OR Boolean!(A=AND Boolean(A=@n5, B=@n3), B=AND Boolean(A=@n3, B=@n4), C:bool=Get bIsTacticalSprinting(), D:bool=Get bIsKneeSliding()) -> (ReturnValue)
   )
   > Set alphaSkirtRigidBody float(alphaSkirtRigidBody float:double=0.5) -> (Output_Get:double)
   n7 = Branch(execute=@n6.else, Condition=Get bIsBaseballSliding())
   > Set alphaSkirtRigidBody float(alphaSkirtRigidBody float=0)
   n8 = Branch(execute=@n7.else, Condition=OR Boolean!(A=Get bIsJumping(), B=Get bIsFalling()) -> (ReturnValue))
   > Set alphaSkirtRigidBody float(alphaSkirtRigidBody float=0.6)
   Set alphaSkirtRigidBody float(execute=@n8.else, alphaSkirtRigidBody float=1)
   n9 = Sequence(execute=@n1.then_2)
   n10 = Get bIsMoving2D()
   n11 = Get bIsJumping()
   n12 = Get bIsFalling()
   > Set jumpingInPlace(jumpingInPlace:bool=AND Boolean(A=OR Boolean(A=@n11, B=@n12), B=NOT Boolean(A=@n10))) -> (Output_Get:bool)
   > Set lowerLOD(lowerLOD:bool="integer >= integer"(A=Get CurrentLOD(), B=2)) -> (Output_Get:bool)
   n13 = Sequence!(execute=@n9.then_2) -> (then_0, then_1)
   > n14 = Branch(Condition=Get bIsInFrontEnd())
   n15 = GetOwnerFrontEndAnimInstance()
   > n16 = Branch(Condition=IsValid(Object=@n15))
   > Set Custom Idle is Playing float(Custom Idle is Playing float:double=GetCurveValue(self=@n15, CurveName=frontend_idle_bypass)) -> (Output_Get:double)
   Set Custom Idle is Playing float(execute=[@n16.else, @n14.else], Custom Idle is Playing float=0)
   n17 = Truncate(A=Get Custom Idle is Playing float() -> (Custom Idle is Playing float:double))
   Set Not Custom Idle is Playing bool(execute=@n13.then_1, Not Custom Idle is Playing bool:bool=NOT Boolean(A="To Boolean (Integer)"!(InInt:int=@n17) -> (ReturnValue:bool))) -> (Output_Get:bool)
   n18 = GetOwnerSkeletalMeshComponent()
   Set rotateCapeSwing alpha(
    execute=@n9.then_1
    rotateCapeSwing alpha:double=float * float(
     A:bool=Get bIsPlayingMeleeAnim()
     B=GetAlphaFromReferenceTransformDelta(
      SkeletalMeshComponent=@n18
      BoneName=spine_05
      BaseName=spine_03
      DeltaComponentType=QuaternionTwist
      InputRange=(X=-5,Y=-20)
      TwistAxis="1,0,0"
     )
    ) {is=title K2Node_PromotableOperator}
   ) -> (Output_Get:double)
   Event OnParachuteOpened()
   > Switch on EGliderType(Selection=Get PlayerGliderType() -> (PlayerGliderType=HangGlider))
   > Set isUsingHangGlider(isUsingHangGlider:bool=true) -> (Output_Get:bool)
   Event OnParachuteClosed()
   > Set isUsingHangGlider(isUsingHangGlider=false)
   Event OnCharacterCustomizationLoaded()
   n19 = CheckForBackbling()
   > Set isNotWearingBlackbling(isNotWearingBlackbling:bool=NOT Boolean(A=OR Boolean!(A=@n19, B=@n19.bIsBackblingACape, C:bool=@n19.bIsBackblingLong) -> (ReturnValue))) -> (Output_Get:bool)
   > n20 = Branch(Condition=Get bIsInFrontEnd())
   > Set SimSpa animDyn output(SimSpa animDyn output:Anim Phys Sim Space Settings Structure=Get SimSpa animDyn FE() -> (SimSpa animDyn FE:Anim Phys Sim Space Settings Structure)) -> (Output_Get:Anim Phys Sim Space Settings Structure)
   > Set SimSpa animDyn Skirt output(SimSpa animDyn Skirt output:Anim Phys Sim Space Settings Structure=Get SimSpa animDyn Skirt FE() -> (SimSpa animDyn Skirt FE:Anim Phys Sim Space Settings Structure)) -> (Output_Get:Anim Phys Sim Space Settings Structure)
   Set SimSpa animDyn output(execute=@n20.else, SimSpa animDyn output=Get SimSpa animDyn default() -> (SimSpa animDyn default:Anim Phys Sim Space Settings Structure))
   > Set SimSpa animDyn Skirt output(SimSpa animDyn Skirt output=Get SimSpa animDyn default())
  ]}
]
variables [
 alphaSkirtRigidBody float {type=number}
 Custom Idle is Playing float {type=number}
 GravityOverride {type=Vector}
 isNotWearingBlackbling {type=boolean}
 isUsingHangGlider {type=boolean}
 jumpingInPlace {type=boolean}
 lowerLOD {type=boolean}
 Not Custom Idle is Playing bool {type=boolean}
 rotateCapeSwing alpha {type=number}
 SimSpa animDyn default {
  type=AnimPhysSimSpaceSettings
  default {maxAngularAcceleration=10000, maxAngularVelocity=10000, simSpaceAngularAlpha=0, externalAngularVelocity=<0 0 0>}
 }
 SimSpa animDyn FE {
  type=AnimPhysSimSpaceSettings
  default {maxAngularAcceleration=1, maxAngularVelocity=6, simSpaceAngularAlpha=1, externalAngularVelocity=<0 0 0>}
 }
 SimSpa animDyn output {
  type=AnimPhysSimSpaceSettings
  default {maxAngularAcceleration=10000, maxAngularVelocity=10000, simSpaceAngularAlpha=0, externalAngularVelocity=<0 0 0>}
 }
 SimSpa animDyn Skirt FE {
  type=AnimPhysSimSpaceSettings
  default {maxAngularAcceleration=1, maxAngularVelocity=5, simSpaceAngularAlpha=0.4, externalAngularVelocity=<0 0 0>}
 }
 SimSpa animDyn Skirt output {
  type=AnimPhysSimSpaceSettings
  default {maxAngularAcceleration=10000, maxAngularVelocity=10000, simSpaceAngularAlpha=0, externalAngularVelocity=<0 0 0>}
 }
]
