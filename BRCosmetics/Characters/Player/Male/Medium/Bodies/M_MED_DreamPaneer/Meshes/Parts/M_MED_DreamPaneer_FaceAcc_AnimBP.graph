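# Anim Blueprint for the M_MED_DreamPaneer face accessory part, adding secondary physics motion to its hair bones.
# Copies the attached parent's pose, then chains nine identically tuned, root-relative AnimDynamics nodes over front and middle hair bones, using angular springs, tight limits and reduced gravity.
# Each update sets a velocity-scale variable to zero while the owner ziplines, otherwise 0.3, but the dynamics nodes show fixed 1,1,1 scales rather than reading it.
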AnimBlueprint /BRCosmetics/Characters/Player/Male/Medium/Bodies/M_MED_DreamPaneer/Meshes/Parts/M_MED_DreamPaneer_FaceAcc_AnimBP
target_skeleton=&/M_MED_DreamPaneer_FaceAcc_Skeleton
graphs [
 AnimGraph {nodes [
   n1 = Anim Dynamics - Bone: dyn_hair_front_01_cr!(
    ComponentPose=Local To Component(LocalPose=Copy Pose From Mesh() {bUseAttachedParent=true})
    Alpha=1
    ComponentLinearAccScale:Vector="1,1,1"
    ComponentLinearVelScale:Vector="1,1,1"
   ) -> (Pose) {
    angularDampingOverride=1
    angularSpringConstant=20000
    bAngularSpring=true
    bOverrideAngularDamping=true
    bUseGravityOverride=true
    gravityScale=0.35
    simulationSpace=RootRelative
    worldSpaceMinimumScale=1
    boundBone {boneName=dyn_hair_front_01_cr}
    componentAppliedLinearAccClamp=<3000 3000 3000>
    componentLinearAccScale=<1 1 1>
    componentLinearVelScale=<1 1 1>
    gravityOverride {z=-980}
    physicsBodyDefinitions [{
      boundBone {boneName=dyn_hair_front_01_cr}
      boxExtents=<30 30 30>
      constraintSetup {angularLimitsMax=<0 0 5>, angularLimitsMin=<0 -15 -5>, angularTarget=<1 0 0>}
      localJointOffset=<4 0 0>
     }]
    simSpaceSettings {maxAngularAcceleration=35, maxAngularVelocity=7, simSpaceAngularAlpha=1}
   }
   n2 = Anim Dynamics - Bone: dyn_hair_front_02_cr!(ComponentPose=@n1, Alpha=1, ComponentLinearAccScale:Vector="1,1,1", ComponentLinearVelScale:Vector="1,1,1") -> (Pose) {
    angularDampingOverride=1
    angularSpringConstant=20000
    bAngularSpring=true
    bOverrideAngularDamping=true
    bUseGravityOverride=true
    gravityScale=0.35
    simulationSpace=RootRelative
    worldSpaceMinimumScale=1
    boundBone {boneName=dyn_hair_front_02_cr}
    componentAppliedLinearAccClamp=<3000 3000 3000>
    componentLinearAccScale=<1 1 1>
    componentLinearVelScale=<1 1 1>
    gravityOverride {z=-980}
    physicsBodyDefinitions [{
      boundBone {boneName=dyn_hair_front_02_cr}
      boxExtents=<30 30 30>
      constraintSetup {angularLimitsMax=<0 0 5>, angularLimitsMin=<0 -15 -5>, angularTarget=<1 0 0>}
      localJointOffset=<4 0 0>
     }]
    simSpaceSettings {maxAngularAcceleration=35, maxAngularVelocity=7, simSpaceAngularAlpha=1}
   }
   n3 = Anim Dynamics - Bone: dyn_hair_front_01_01_cr!(ComponentPose=@n2, Alpha=1, ComponentLinearAccScale:Vector="1,1,1", ComponentLinearVelScale:Vector="1,1,1") -> (Pose) {
    angularDampingOverride=1
    angularSpringConstant=20000
    bAngularSpring=true
    bOverrideAngularDamping=true
    bUseGravityOverride=true
    gravityScale=0.35
    simulationSpace=RootRelative
    worldSpaceMinimumScale=1
    boundBone {boneName=dyn_hair_front_01_01_cr}
    componentAppliedLinearAccClamp=<3000 3000 3000>
    componentLinearAccScale=<1 1 1>
    componentLinearVelScale=<1 1 1>
    gravityOverride {z=-980}
    physicsBodyDefinitions [{
      boundBone {boneName=dyn_hair_front_01_01_cr}
      boxExtents=<30 30 30>
      constraintSetup {angularLimitsMax=<0 0 5>, angularLimitsMin=<0 -15 -5>, angularTarget=<1 0 0>}
      localJointOffset=<4 0 0>
     }]
    simSpaceSettings {maxAngularAcceleration=35, maxAngularVelocity=7, simSpaceAngularAlpha=1}
   }
   n4 = Anim Dynamics - Bone: dyn_hair_front_01_02_cr!(ComponentPose=@n3, Alpha=1, ComponentLinearAccScale:Vector="1,1,1", ComponentLinearVelScale:Vector="1,1,1") -> (Pose) {
    angularDampingOverride=1
    angularSpringConstant=20000
    bAngularSpring=true
    bOverrideAngularDamping=true
    bUseGravityOverride=true
    gravityScale=0.35
    simulationSpace=RootRelative
    worldSpaceMinimumScale=1
    boundBone {boneName=dyn_hair_front_01_02_cr}
    componentAppliedLinearAccClamp=<3000 3000 3000>
    componentLinearAccScale=<1 1 1>
    componentLinearVelScale=<1 1 1>
    gravityOverride {z=-980}
    physicsBodyDefinitions [{
      boundBone {boneName=dyn_hair_front_01_02_cr}
      boxExtents=<30 30 30>
      constraintSetup {angularLimitsMax=<0 0 5>, angularLimitsMin=<0 -15 -5>, angularTarget=<1 0 0>}
      localJointOffset=<4 0 0>
     }]
    simSpaceSettings {maxAngularAcceleration=35, maxAngularVelocity=7, simSpaceAngularAlpha=1}
   }
   n5 = Anim Dynamics - Bone: dyn_hair_middle_01_cr!(ComponentPose=@n4, Alpha=1, ComponentLinearAccScale:Vector="1,1,1", ComponentLinearVelScale:Vector="1,1,1") -> (Pose) {
    angularDampingOverride=1
    angularSpringConstant=20000
    bAngularSpring=true
    bOverrideAngularDamping=true
    bUseGravityOverride=true
    gravityScale=0.35
    simulationSpace=RootRelative
    worldSpaceMinimumScale=1
    boundBone {boneName=dyn_hair_middle_01_cr}
    componentAppliedLinearAccClamp=<3000 3000 3000>
    componentLinearAccScale=<1 1 1>
    componentLinearVelScale=<1 1 1>
    gravityOverride {z=-980}
    physicsBodyDefinitions [{
      boundBone {boneName=dyn_hair_middle_01_cr}
      boxExtents=<30 30 30>
      constraintSetup {angularLimitsMax=<0 0 5>, angularLimitsMin=<0 -15 -5>, angularTarget=<1 0 0>}
      localJointOffset=<4 0 0>
     }]
    simSpaceSettings {maxAngularAcceleration=35, maxAngularVelocity=7, simSpaceAngularAlpha=1}
   }
   n6 = Anim Dynamics - Bone: dyn_hair_middle_01_lf!(ComponentPose=@n5, Alpha=1, ComponentLinearAccScale:Vector="1,1,1", ComponentLinearVelScale:Vector="1,1,1") -> (Pose) {
    angularDampingOverride=1
    angularSpringConstant=20000
    bAngularSpring=true
    bOverrideAngularDamping=true
    bUseGravityOverride=true
    gravityScale=0.35
    simulationSpace=RootRelative
    worldSpaceMinimumScale=1
    boundBone {boneName=dyn_hair_middle_01_lf}
    componentAppliedLinearAccClamp=<3000 3000 3000>
    componentLinearAccScale=<1 1 1>
    componentLinearVelScale=<1 1 1>
    gravityOverride {z=-980}
    physicsBodyDefinitions [{
      boundBone {boneName=dyn_hair_middle_01_lf}
      boxExtents=<30 30 30>
      constraintSetup {angularLimitsMax=<0 0 5>, angularLimitsMin=<0 -15 -5>, angularTarget=<1 0 0>}
      localJointOffset=<4 0 0>
     }]
    simSpaceSettings {maxAngularAcceleration=35, maxAngularVelocity=7, simSpaceAngularAlpha=1}
   }
   n7 = Anim Dynamics - Bone: dyn_hair_front_01_rt!(ComponentPose=@n6, Alpha=1, ComponentLinearAccScale:Vector="1,1,1", ComponentLinearVelScale:Vector="1,1,1") -> (Pose) {
    angularDampingOverride=1
    angularSpringConstant=20000
    bAngularSpring=true
    bOverrideAngularDamping=true
    bUseGravityOverride=true
    gravityScale=0.35
    simulationSpace=RootRelative
    worldSpaceMinimumScale=1
    boundBone {boneName=dyn_hair_front_01_rt}
    componentAppliedLinearAccClamp=<3000 3000 3000>
    componentLinearAccScale=<1 1 1>
    componentLinearVelScale=<1 1 1>
    gravityOverride {z=-980}
    physicsBodyDefinitions [{
      boundBone {boneName=dyn_hair_front_01_rt}
      boxExtents=<30 30 30>
      constraintSetup {angularLimitsMax=<0 0 5>, angularLimitsMin=<0 -15 -5>, angularTarget=<1 0 0>}
      localJointOffset=<4 0 0>
     }]
    simSpaceSettings {maxAngularAcceleration=35, maxAngularVelocity=7, simSpaceAngularAlpha=1}
   }
   n8 = Anim Dynamics - Bone: dyn_hair_front_01_lf!(ComponentPose=@n7, Alpha=1, ComponentLinearAccScale:Vector="1,1,1", ComponentLinearVelScale:Vector="1,1,1") -> (Pose) {
    angularDampingOverride=1
    angularSpringConstant=20000
    bAngularSpring=true
    bOverrideAngularDamping=true
    bUseGravityOverride=true
    gravityScale=0.35
    simulationSpace=RootRelative
    worldSpaceMinimumScale=1
    boundBone {boneName=dyn_hair_front_01_lf}
    componentAppliedLinearAccClamp=<3000 3000 3000>
    componentLinearAccScale=<1 1 1>
    componentLinearVelScale=<1 1 1>
    gravityOverride {z=-980}
    physicsBodyDefinitions [{
      boundBone {boneName=dyn_hair_front_01_lf}
      boxExtents=<30 30 30>
      constraintSetup {angularLimitsMax=<0 0 5>, angularLimitsMin=<0 -15 -5>, angularTarget=<1 0 0>}
      localJointOffset=<4 0 0>
     }]
    simSpaceSettings {maxAngularAcceleration=35, maxAngularVelocity=7, simSpaceAngularAlpha=1}
   }
   n9 = Anim Dynamics - Bone: dyn_hair_middle_02_cr!(ComponentPose=@n8, Alpha=1, ComponentLinearAccScale:Vector="1,1,1", ComponentLinearVelScale:Vector="1,1,1") -> (Pose) {
    angularDampingOverride=1
    angularSpringConstant=20000
    bAngularSpring=true
    bOverrideAngularDamping=true
    bUseGravityOverride=true
    gravityScale=0.35
    simulationSpace=RootRelative
    worldSpaceMinimumScale=1
    boundBone {boneName=dyn_hair_middle_02_cr}
    componentAppliedLinearAccClamp=<3000 3000 3000>
    componentLinearAccScale=<1 1 1>
    componentLinearVelScale=<1 1 1>
    gravityOverride {z=-980}
    physicsBodyDefinitions [{
      boundBone {boneName=dyn_hair_middle_02_cr}
      boxExtents=<30 30 30>
      constraintSetup {angularLimitsMax=<0 0 5>, angularLimitsMin=<0 -15 -5>, angularTarget=<1 0 0>}
      localJointOffset=<4 0 0>
     }]
    simSpaceSettings {maxAngularAcceleration=35, maxAngularVelocity=7, simSpaceAngularAlpha=1}
   }
   Output Pose(Result=Component To Local(ComponentPose=@n9)) {layerGroup=None}
  ]}
 EventGraph {nodes [
   Event BlueprintUpdateAnimation()
   n1 = GetOwnerAnimationInterface!(self:Custom Character Part Anim Instance Object Reference) -> (ReturnValue:Fort Player Animation Interface Object Reference)
   > Is Valid(InputObject=@n1)
   > n2 = Branch(Condition=Break Fort Anim Input Zipline!(FortAnimInput_Zipline:"Fort Anim Input Zipline Structure (by ref)"=GetZiplineInput!(self:Fort Player Animation Interface Object Reference=@n1) -> (ReturnValue:"Fort Anim Input Zipline Structure (by ref)")) -> (
     bIsZiplining:bool=~
     bIsOnAscender:bool=~
     bShouldPlayPivotTransition:bool=~
     LeanYaw:float=~
     PivotBlendDelayRemaining:float=~
     PivotBlendDelay:float=~
     PivotCardinalDirection:EFortCardinalDirection Enum
     WorldVelocityLastTick:Vector=~
    ) {is=title K2Node_BreakStruct})
   > Set Componen Linear Vel Scale(Componen Linear Vel Scale:Vector="0, 0, 0") -> (Output_Get:Vector)
   Set Componen Linear Vel Scale(execute=@n2.else, Componen Linear Vel Scale="0.3,0.3,0.3")
  ]}
]
variables [Componen Linear Vel Scale {type=Vector, default=<0.3 0.3 0.3>}]
